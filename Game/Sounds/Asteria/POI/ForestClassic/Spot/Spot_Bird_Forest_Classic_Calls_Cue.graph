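# Sound cue for quiet one-shot spot bird calls in the Asteria classic forest POI ambience, using a bird attenuation and the outdoor ambient one-shot sound class.
# A quality-level node sends the first two levels to a random pool of 32 calls and the last two to a four-call subset excluded from branch culling.
# Each pool passes through a modulator that varies volume and pitch.
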SoundCue /Game/Sounds/Asteria/POI/ForestClassic/Spot/Spot_Bird_Forest_Classic_Calls_Cue
root=QualityLevel_0
nodes [
 {id=Modulator_1, children [Random_2], properties {pitchMax=1.1, volumeMax=1.1, volumeMin=0.9}}
 {id=Modulator_2, children [Random_3], properties {pitchMax=1.1, volumeMax=1.1, volumeMin=0.9}}
 {id=QualityLevel_0, children [Modulator_1,Modulator_1,Modulator_2,Modulator_2]}
 {
  id=Random_2
  children [
   WavePlayer_9
   WavePlayer_10
   WavePlayer_11
   WavePlayer_12
   WavePlayer_13
   WavePlayer_14
   WavePlayer_15
   WavePlayer_16
   WavePlayer_17
   WavePlayer_18
   WavePlayer_19
   WavePlayer_20
   WavePlayer_21
   WavePlayer_22
   WavePlayer_23
   WavePlayer_24
   WavePlayer_25
   WavePlayer_26
   WavePlayer_27
   WavePlayer_28
   WavePlayer_29
   WavePlayer_30
   WavePlayer_31
   WavePlayer_32
   WavePlayer_33
   WavePlayer_34
   WavePlayer_35
   WavePlayer_36
   WavePlayer_37
   WavePlayer_38
   WavePlayer_39
   WavePlayer_40
  ]
  properties {weights [1,1,1,1,1,1,1,1,1,1,1,1,1,1,1,1,1,1,1,1,1,1,1,1,1,1,1,1,1,1,1,1]}
 }
 {
  id=Random_3
  children [WavePlayer_42, WavePlayer_43, WavePlayer_44, WavePlayer_45]
  properties {bShouldExcludeFromBranchCulling=true, weights [1,1,1,1]}
 }
 {id=WavePlayer_10, properties {soundWaveAssetPtr=&/Spot_Bird_Forest_Classic_Call_02}}
 {id=WavePlayer_11, properties {soundWaveAssetPtr=&/Spot_Bird_Forest_Classic_Call_03}}
 {id=WavePlayer_12, properties {soundWaveAssetPtr=&/Spot_Bird_Forest_Classic_Call_04}}
 {id=WavePlayer_13, properties {soundWaveAssetPtr=&/Spot_Bird_Forest_Classic_Call_05}}
 {id=WavePlayer_14, properties {soundWaveAssetPtr=&/Spot_Bird_Forest_Classic_Call_06}}
 {id=WavePlayer_15, properties {soundWaveAssetPtr=&/Spot_Bird_Forest_Classic_Call_07}}
 {id=WavePlayer_16, properties {soundWaveAssetPtr=&/Spot_Bird_Forest_Classic_Call_08}}
 {id=WavePlayer_17, properties {soundWaveAssetPtr=&/Spot_Bird_Forest_Classic_Call_09}}
 {id=WavePlayer_18, properties {soundWaveAssetPtr=&/Spot_Bird_Forest_Classic_Call_10}}
 {id=WavePlayer_19, properties {soundWaveAssetPtr=&/Spot_Bird_Forest_Classic_Call_11}}
 {id=WavePlayer_20, properties {soundWaveAssetPtr=&/Spot_Bird_Forest_Classic_Call_12}}
 {id=WavePlayer_21, properties {soundWaveAssetPtr=&/Spot_Bird_Forest_Classic_Call_13}}
 {id=WavePlayer_22, properties {soundWaveAssetPtr=&/Spot_Bird_Forest_Classic_Call_14}}
 {id=WavePlayer_23, properties {soundWaveAssetPtr=&/Spot_Bird_Forest_Classic_Call_15}}
 {id=WavePlayer_24, properties {soundWaveAssetPtr=&/Spot_Bird_Forest_Classic_Call_16}}
 {id=WavePlayer_25, properties {soundWaveAssetPtr=&/Spot_Bird_Forest_Classic_Call_17}}
 {id=WavePlayer_26, properties {soundWaveAssetPtr=&/Spot_Bird_Forest_Classic_Call_18}}
 {id=WavePlayer_27, properties {soundWaveAssetPtr=&/Spot_Bird_Forest_Classic_Call_19}}
 {id=WavePlayer_28, properties {soundWaveAssetPtr=&/Spot_Bird_Forest_Classic_Call_20}}
 {id=WavePlayer_29, properties {soundWaveAssetPtr=&/Spot_Bird_Forest_Classic_Call_21}}
 {id=WavePlayer_30, properties {soundWaveAssetPtr=&/Spot_Bird_Forest_Classic_Call_22}}
 {id=WavePlayer_31, properties {soundWaveAssetPtr=&/Spot_Bird_Forest_Classic_Call_23}}
 {id=WavePlayer_32, properties {soundWaveAssetPtr=&/Spot_Bird_Forest_Classic_Call_24}}
 {id=WavePlayer_33, properties {soundWaveAssetPtr=&/Spot_Bird_Forest_Classic_Call_25}}
 {id=WavePlayer_34, properties {soundWaveAssetPtr=&/Spot_Bird_Forest_Classic_Call_26}}
 {id=WavePlayer_35, properties {soundWaveAssetPtr=&/Spot_Bird_Forest_Classic_Call_27}}
 {id=WavePlayer_36, properties {soundWaveAssetPtr=&/Spot_Bird_Forest_Classic_Call_28}}
 {id=WavePlayer_37, properties {soundWaveAssetPtr=&/Spot_Bird_Forest_Classic_Call_29}}
 {id=WavePlayer_38, properties {soundWaveAssetPtr=&/Spot_Bird_Forest_Classic_Call_30}}
 {id=WavePlayer_39, properties {soundWaveAssetPtr=&/Spot_Bird_Forest_Classic_Call_31}}
 {id=WavePlayer_40, properties {soundWaveAssetPtr=&/Spot_Bird_Forest_Classic_Call_32}}
 {id=WavePlayer_42, properties {soundWaveAssetPtr=&/Spot_Bird_Forest_Classic_Call_01}}
 {id=WavePlayer_43, properties {soundWaveAssetPtr=&Spot_Bird_Forest_Classic_Call_11}}
 {id=WavePlayer_44, properties {soundWaveAssetPtr=&Spot_Bird_Forest_Classic_Call_12}}
 {id=WavePlayer_45, properties {soundWaveAssetPtr=&Spot_Bird_Forest_Classic_Call_26}}
 {id=WavePlayer_9, properties {soundWaveAssetPtr=&Spot_Bird_Forest_Classic_Call_01}}
]
properties {
 attenuationSettings=&Fort_TileAmbient_OneShot_Bird_Att
 bEnableBusSends=false
 maxDistance=8005
 soundClassObject=&Fort_Ambient_OneShots_Outdoors
 volumeMultiplier=0.15
 preEffectBusSends [{}]
}
